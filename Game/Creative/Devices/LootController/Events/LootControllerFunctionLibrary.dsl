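; Blueprint function library for the Creative loot controller device that broadcasts an event when a building container is opened.
; MakeAndSendOpenContainerEvent gets the container's searching pawn and builds an OpenContainerEvent from that pawn's controller and the given container type enum.
; It broadcasts the event through the GameplayEventRouter obtained for the container's FortPlayspace.
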(fn MakeAndSendOpenContainerEvent
  (params (Container :type Building%20Container%20Object%20Reference) (ContainerType :type ContainerEventTypes%20Enum :default "NewEnumerator0"))
  (bind _returnvalue (Container|GetSearchingPawn Container))
  (bind _returnvalue_2 (GameplayEventRouter|GetEventRouter "/Script/FortniteGame.FortPlayspace" Container))
  (GameplayEventRouter|BroadcastEvent _returnvalue_2 (Utilities|Struct|MakeOpenContainerEvent (Pawn|GetController _returnvalue) ContainerType) 0))
